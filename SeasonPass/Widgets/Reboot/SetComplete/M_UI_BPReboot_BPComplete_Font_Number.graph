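# UI material for the number text on the Reboot season pass set-complete screen, styling font glyphs with gradient color and glowing outlines.
# Rotated UV gradients blend three color parameters, while a rotated smoothstep mask shapes the fill alpha.
# Two outline passes on the squared font signed distance form sharp and soft edges, weighted by distance from Number Appear.
# A time-panned, screen-space blurred noise sample provides a distortion signal.
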Material /SeasonPass/Widgets/Reboot/SetComplete/M_UI_BPReboot_BPComplete_Font_Number
graph {
 properties {
  bAutomaticallySetUsageInEditor=false
  bUsedWithStaticMesh=false
  blendMode=BLEND_Translucent
  materialDomain=MD_UI
  usedShadingModels=Default Lit
 }
 nodes [
  n1 = ScalarParameter() {parameterName=Fill Alpha}
  n2 = NamedRerouteUsage()
  n3 = ComponentMask(None=MaterialFunctionCall(UVs=TextureCoordinate(), "Rotation Angle (0-1)"=Constant() {r=0.015}) {materialFunction=&CustomRotator}.Rotated Values) {g=true}
  LinearInterpolate(B=Add(A=Multiply(A=@n1, B=@n2), B=NamedRerouteUsage()), Alpha=SmoothStep(Value=OneMinus(None=@n3)) {constMax=0.5, constMin=0.2}) {constB=0}
  NamedRerouteDeclaration(None=Multiply(A=Time(), B=Constant() {r=0.6})) {name=Time, nodeColor=<rgba 0 0 0 1>}
  n4 = Panner(
   Coordinate=MaterialFunctionCall(UVs=ScreenPosition().ViewportUV, "Rotation Angle (0-1)"=Constant() {r=0.125}) {materialFunction=&CustomRotator}.Rotated Values
   Time=NamedRerouteUsage()
  ) {speedX=-0.2, speedY=0.1}
  n5 = Constant2Vector() {g=3, r=1}
  NamedRerouteDeclaration(None=TextureSample(UVs=MaterialFunctionCall(UVs=@n4, Texture Scale=@n5) {materialFunction=&ScaleUVsByCenter}.UVs) {samplerType=SAMPLERTYPE_Grayscale, texture=&LowResBlurredNoise}.R) {name=Distortion, nodeColor=<rgba 0 0 0 1>}
  n6 = VectorParameter() {parameterName=Color 2, defaultValue=<rgba 0.2582 0.5711 1 1>}
  n7 = VectorParameter() {parameterName=Color 3, defaultValue=<rgba 0.6834 0.8258 1 1>}
  n8 = Constant() {r=0.5}
  n9 = TextureCoordinate()
  n10 = Constant() {r=0.7}
  n11 = SmoothStep(
   Min=Subtract(B=@n8) {constA=0.5}
   Max=Add(B=@n8) {constA=0.5}
   Value=ComponentMask(None=MaterialFunctionCall(UVs=@n9, "Rotation Angle (0-1)"=@n10) {materialFunction=&CustomRotator}.Rotated Values) {g=true}
  )
  n12 = ScalarParameter() {defaultValue=0.5, parameterName=WhiteGradient_Opacity}
  n13 = Power(Base=SmoothStep(Value=Multiply(A=@n11, B=@n12) {constB=0.5}) {constMin=0.25}) {constExponent=0.5}
  n14 = Constant() {r=0.5}
  n15 = Subtract(B=@n14) {constA=0.5}
  n16 = Add(B=@n14) {constA=0.5}
  n17 = ComponentMask(None=MaterialFunctionCall(UVs=TextureCoordinate(), "Rotation Angle (0-1)"=Constant() {r=0.8}) {materialFunction=&CustomRotator}.Rotated Values) {g=true}
  n18 = SmoothStep(Value=Multiply(A=SmoothStep(Min=@n15, Max=@n16, Value=@n17), B=ScalarParameter() {defaultValue=0.5, parameterName=WhiteGradient_Opacity}) {constB=0.5}) {constMax=0.5}
  NamedRerouteDeclaration(None=LinearInterpolate(
    A=VectorParameter() {parameterName=Color 1, defaultValue=<rgba 0.0395 1 0.2307 1>}.RGB
    B=LinearInterpolate(A=@n6.RGB, B=@n7.RGB, Alpha=@n13)
    Alpha=Power(Base=@n18) {constExponent=0.5}
   )) {name=Color, nodeColor=<rgba 0 0 0 1>}
  n19 = FontSignedDistance()
  n20 = Multiply(A=@n19.Signed Distance, B=@n19.Signed Distance)
  n21 = MaterialFunctionCall!(
   DistanceField=@n20
   Threshold
   Thickness=Multiply(A=@n19.Pixel Distance Factor, B=Constant() {r=0.0015})
   Alignment=Constant() {r=2}
   Smoothing=Constant() {r=3}
  ) {materialFunction=/FortUILibrary/MaterialFunctions/Shapes/Modifiers/MF_UI_ShapeOutline}
  n22 = MaterialFunctionCall(
   DistanceField=@n20
   Thickness=Multiply(A=@n19.Pixel Distance Factor, B=Constant() {r=0.004})
   Alignment=Constant() {r=1.5}
   Smoothing=Constant() {r=16}
  ) {materialFunction=&MF_UI_ShapeOutline}
  n23 = NamedRerouteDeclaration(None=Add(A=Multiply(A=@n21.Outline), B=Multiply(A=@n22.Outline) {constB=0.4})) {name=Outline Alpha, nodeColor=<rgba 1 0.6563 0 1>}
  n24 = Multiply(A=NamedRerouteUsage()) {constB=3}
  n25 = MaterialFunctionCall() {materialFunction=&GetUserInterfaceUV}
  n26 = MaterialFunctionCall(
   UVs=AppendVector(A=ComponentMask(None=@n25.9-Slice UV) {r=true}, B=ComponentMask(None=@n25.9-Slice UV) {g=true})
   Texture Scale=Constant2Vector() {g=1, r=1}
  ) {materialFunction=&ScaleUVsByCenter}
  n27 = ScalarParameter() {defaultValue=0.5, parameterName=Number Appear}
  n28 = Power(Base=OneMinus(None=Distance(A=@n26.UVs, B=@n27))) {constExponent=4}
  n29 = ScalarParameter() {defaultValue=0.5, parameterName=Line Alpha Opacity}
  NamedRerouteDeclaration(None=LinearInterpolate(B=Multiply(A=@n23, B=@n24), Alpha=Multiply(A=@n28, B=@n29) {constB=0.3})) {name=Line Alpha, nodeColor=<rgba 0 0 0 1>}
  NamedRerouteUsage()
 ]
}
