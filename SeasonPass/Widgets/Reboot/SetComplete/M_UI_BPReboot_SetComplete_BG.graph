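# Translucent UI material, evidently the background for the Season Pass Reboot set-complete screen, layering colored gradients and soft glow blobs.
# A rotated, eased gradient varies Color 4 for the base, while a panned, mirrored gradient scrolls Color 1 and Color 2 as the rift glow.
# Distance-based radial falloffs form center, inverted-center and lateral blobs.
# Named reroutes carry these layers into lerps with Color 3 and an opacity add.
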Material /SeasonPass/Widgets/Reboot/SetComplete/M_UI_BPReboot_SetComplete_BG
graph {
 properties {bAutomaticallySetUsageInEditor=false, blendMode=BLEND_Translucent, materialDomain=MD_UI, usedShadingModels=Default Lit}
 nodes [
  Add(A=ScalarParameter() {defaultValue=1, parameterName=Back Opacity}, B=NamedRerouteUsage())
  LinearInterpolate(
   A=LinearInterpolate(A=NamedRerouteUsage(), B=NamedRerouteUsage(), Alpha=NamedRerouteUsage()) {constB=0}
   B=VectorParameter() {parameterName=Color 3, defaultValue=<rgba 0.1024 0.5312 0.1878 1>}.RGB
   Alpha=Add(A=NamedRerouteUsage(), B=NamedRerouteUsage())
  )
  n1 = VectorParameter() {parameterName=Color 4, defaultValue=<rgba 0.0091 0.0122 0.05 1>}
  n2 = MaterialFunctionCall(UVs=TextureCoordinate(), GradientStart=Constant() {r=-0.2}, GradientEnd=Constant() {r=1.2}, RotationAngle=Constant() {r=0.3}) {materialFunction=&MF_UI_LinearGradient}
  NamedRerouteDeclaration(None=LinearInterpolate(
    A=Power(Base=@n1.RGB) {constExponent=1.25}
    B=@n1.RGB
    Alpha=MaterialFunctionCall("Value (0-1)"=@n2.Result) {materialFunction=&MF_UI_EaseCurves}.EaseInOutSine
   )) {name=Base Gradient, nodeColor=<rgba 0 0 0 1>}
  n3 = VectorParameter() {group=0 - Coloring, parameterName=Color 1, defaultValue=<rgba 0.3774 0.0998 0.7708 1>}
  n4 = MaterialFunctionCall(UVs=TextureCoordinate(), Texture Scale=Constant() {r=2}) {materialFunction=&ScaleUVsByCenter}
  n5 = Frac(None=MaterialFunctionCall(UV Channel=Panner(Coordinate=@n4.UVs) {speedX=-0.2}) {materialFunction=&LinearGradient}.UGradient)
  n6 = Constant() {r=0.5}
  n7 = MaterialFunctionCall("Value (0-1)"=Multiply(A=Distance(A=@n5, B=@n6)) {constB=2}) {materialFunction=&MF_UI_EaseCurves}
  n8 = VectorParameter() {group=0 - Coloring, parameterName=Color 2, defaultValue=<rgba 0.0975 0.4718 0.6458 1>}
  NamedRerouteDeclaration(None=Multiply(A=LinearInterpolate(A=@n3.RGB, B=@n8.RGB, Alpha=@n7.EaseInOutCubic), B=ScalarParameter() {defaultValue=1, parameterName=UnderGlow_Opacity})) {name=Rift Gradient, nodeColor=<rgba 0 0 0 1>}
  n9 = MaterialFunctionCall() {materialFunction=&GetUserInterfaceUV}
  n10 = Add(A=@n9.9-Slice UV, B=ScalarParameter() {parameterName=Centre Blob Position}) {constB=0}
  n11 = MaterialFunctionCall(
   UVs=AppendVector(A=ComponentMask(None=@n9.9-Slice UV) {r=true}, B=ComponentMask(None=@n10) {g=true})
   Texture Scale=Multiply(A=Constant2Vector() {g=1, r=0.75}, B=ScalarParameter() {defaultValue=1, parameterName=Centre Blob Scale})
  ) {materialFunction=&ScaleUVsByCenter}
  n12 = Constant() {r=0.5}
  n13 = Power(Base=OneMinus(None=Distance(A=@n11.UVs, B=@n12))) {constExponent=4}
  NamedRerouteDeclaration(None=Multiply(A=@n13, B=ScalarParameter() {defaultValue=0.1, parameterName=Center Blob Opacity}) {constB=0.3}) {name=Center Bloob, nodeColor=<rgba 0 0 0 1>}
  NamedRerouteDeclaration(None=Multiply(A=ScalarParameter() {parameterName=Inverted Center Blob Opacity}, B=OneMinus(None=@n13))) {name=Inverted Center Bloob, nodeColor=<rgba 0 0 0 1>}
  n14 = MaterialFunctionCall(
   UVs=MaterialFunctionCall() {materialFunction=&GetUserInterfaceUV}.Normalized UV
   Texture Scale=Constant2Vector() {g=2, r=1}
  ) {materialFunction=&ScaleUVsByCenter}
  n15 = AppendVector(A=ComponentMask(None=@n14.UVs) {r=true}, B=ComponentMask(None=@n14.UVs) {g=true})
  n16 = Constant()
  n17 = Power(Base=OneMinus(None=Distance(A=@n15, B=@n16))) {constExponent=4}
  n18 = ComponentMask(None=@n14.UVs) {g=true}
  n19 = AppendVector(A=ComponentMask(None=@n14.UVs) {r=true}, B=@n18)
  n20 = Constant() {r=1}
  n21 = Power(Base=OneMinus(None=Distance(A=@n19, B=@n20))) {constExponent=3}
  NamedRerouteDeclaration(None=Multiply(A=LinearInterpolate(A=@n17, Alpha=@n21), B=ScalarParameter() {defaultValue=0.35, parameterName=Blobs Opacity}) {constB=0.3}) {name=Lateral Bloobs, nodeColor=<rgba 0 0 0 1>}
  OneMinus(None=@n18)
 ]
}
