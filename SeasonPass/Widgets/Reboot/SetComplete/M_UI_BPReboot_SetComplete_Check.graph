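# Translucent UI material that draws a circled checkmark for the BPReboot set-complete widget in the season pass.
# Circle functions build the fill and outline, while an MSDF icon supplies the check's fill and layered outline, with parameters controlling opacities.
# A rotated, mirrored radial arc, hardened with Ceil, provides a parameterized progress mask.
# An eased, spinning three-color gradient and constant green feed the compositing through named reroutes.
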Material /SeasonPass/Widgets/Reboot/SetComplete/M_UI_BPReboot_SetComplete_Check
graph {
 properties {bAutomaticallySetUsageInEditor=false, blendMode=BLEND_Translucent, materialDomain=MD_UI, usedShadingModels=Default Lit}
 nodes [
  n1 = Constant3Vector() {constant=<rgba 0.0395 1 0.2307 1>}
  n2 = NamedRerouteUsage()
  Add(
   A=Multiply(A=NamedRerouteUsage(), B=ScalarParameter() {parameterName=Gradient Amount}) {constB=0}
   B=LinearInterpolate(A=LinearInterpolate(B=@n1, Alpha=@n2), B=Constant(), Alpha=NamedRerouteUsage())
  )
  n3 = NamedRerouteUsage()
  n4 = NamedRerouteUsage()
  Add(A=Add(A=NamedRerouteUsage(), B=LinearInterpolate(B=@n3, Alpha=@n4)), B=NamedRerouteUsage())
  n5 = TextureCoordinate()
  n6 = Constant() {r=0.8}
  n7 = MaterialFunctionCall(UVs=@n5, Size=@n6, OutlineThickness=Constant() {r=0.015}) {materialFunction=&MF_UI_Circle}
  NamedRerouteDeclaration(None=LinearInterpolate(B=ScalarParameter() {defaultValue=1, parameterName=FinalOpacity}, Alpha=@n7.Fill)) {name=CircleFIll, nodeColor=<rgba 0.3359 0 1 1>}
  n8 = Multiply(A=MaterialFunctionCall(UVs=@n5, Size=@n6, OutlineThickness=Constant() {r=0.01}, GlowMax=Constant() {r=0.03}) {materialFunction=&MF_UI_Circle}.Outline) {constB=0.5}
  NamedRerouteDeclaration(None=Multiply(A=LinearInterpolate(A=@n8, Alpha=@n7.Outline), B=ScalarParameter() {defaultValue=1, parameterName=CircleOutlineOpacity})) {name=CircleOutline, nodeColor=<rgba 0 1 0.0156 1>}
  n9 = MaterialFunctionCall(UVs=TextureCoordinate(), "Rotation Angle (0-1)"=ScalarParameter() {parameterName=ArcEndAngle}) {materialFunction=&CustomRotator}
  n10 = AppendVector(A=OneMinus(None=ComponentMask(None=@n9.Rotated Values) {r=true}), B=ComponentMask(None=@n9.Rotated Values) {g=true})
  n11 = ScalarParameter() {parameterName=ArcStartAngle}
  n12 = Constant() {r=1}
  NamedRerouteDeclaration(None=Ceil(None=MaterialFunctionCall!(UV=@n10, "Arc Start Angle (0-1)"=@n11, "Arc End Angle (0-1)"=@n12, "Fill Ratio (0-1)", Fill Clockwise) {materialFunction=/FortUILibrary/MaterialFunctions/Basics/MF_UI_RadialArc}.Radial Fill)) {name=Progress Mask, nodeColor=<rgba 0 0 0 1>}
  n13 = Multiply(A=TextureCoordinate()) {constB=1.7}
  n14 = MaterialFunctionCall!(MSDF RGB=TextureSample(UVs=Add(A=@n13) {constB=-0.35}) {samplerType=SAMPLERTYPE_LinearColor, texture=/FortUILibrary/Textures/Icons/T_UI_Icon_UIKit_Check_MSDF}.RGB) {materialFunction=/FortUILibrary/MaterialFunctions/Shapes/Modifiers/MF_UI_MSDFtoSDF}
  n15 = Constant() {r=0.5}
  n16 = Add(
   A=MaterialFunctionCall!(DistanceField=@n14.SDF, Threshold=@n15, Thickness=Constant() {r=0.1}, Alignment, Smoothing) {materialFunction=/FortUILibrary/MaterialFunctions/Shapes/Modifiers/MF_UI_ShapeOutline}.Outline
   B=MaterialFunctionCall(DistanceField=@n14.SDF, Threshold=@n15, Thickness=Constant() {r=0.012}, Smoothing=Constant() {r=0.44}) {materialFunction=&MF_UI_ShapeOutline}.Outline
  )
  NamedRerouteDeclaration(None=Multiply(A=Saturate(None=@n16), B=ScalarParameter() {defaultValue=1, parameterName=CheckedOutlineOpacity})) {name=CheckedOutline, nodeColor=<rgba 0 1 0.625 1>}
  NamedRerouteDeclaration(None=MaterialFunctionCall(DistanceField=@n14.SDF, Threshold=@n15) {materialFunction=&MF_UI_AntiAliasing}.Outer) {name=CheckedFill, nodeColor=<rgba 1 0.3047 0 1>}
  n17 = MaterialFunctionCall("Value (0-1)"=MaterialFunctionCall(UV Channel=Rotator() {speed=2.5}) {materialFunction=&LinearGradient}.UGradient) {materialFunction=&MF_UI_EaseCurves}
  n18 = VectorParameter() {group=0 - Coloring, parameterName=Color 2, defaultValue=<rgba 0.6883 0.3193 1 1>}
  n19 = VectorParameter() {group=0 - Coloring, parameterName=Color 1, defaultValue=<rgba 0.1615 0.4154 1 1>}
  NamedRerouteDeclaration(None=LinearInterpolate(
    A=VectorParameter() {parameterName=Color 3, defaultValue=<rgba 0.0395 1 0.2307 1>}.RGB
    B=LinearInterpolate(A=@n19.RGB, B=@n18.RGB, Alpha=@n17.EaseInOutCubic)
    Alpha=@n17.EaseOutCubic
   )) {name=Rift Gradient, nodeColor=<rgba 0 0 0 1>}
 ]
}
