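# Translucent UI material for the Season Pass Reboot "set complete" modal, drawing its backdrop, glows and accent line.
# A dark base gradient is built alongside a panning two-color rift gradient, color-dodged underglow, center glow, green tint and soft corner blobs.
# Named reroutes carry these layers into modal color, line and mask sections.
# Reveal parameters drive the modal's vertical scale and the line's wipe, while procedural masks fade the edges.
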Material /SeasonPass/Widgets/Reboot/SetComplete/M_UI_BPReboot_SetComplete_Modal
graph {
 properties {
  bAutomaticallySetUsageInEditor=false
  bUsedWithStaticMesh=false
  blendMode=BLEND_Translucent
  materialDomain=MD_UI
  usedShadingModels=Default Lit
 }
 nodes [
  MaterialFunctionCall() {materialFunction=&DebugTimeSine}
  MaterialFunctionCall(
   Hue Shift Percentage=Multiply(A=MaterialFunctionCall!(UV, Seed) {materialFunction=&MF_UI_Random}.Result) {constB=0.035}
   Texture=LinearInterpolate(A=NamedRerouteUsage(), B=NamedRerouteUsage(), Alpha=NamedRerouteUsage())
  ) {materialFunction=&HueShift}
  n1 = MaterialFunctionCall() {materialFunction=&GetUserInterfaceUV}
  n2 = Distance(A=ComponentMask(None=@n1.Normalized UV) {r=true}, B=Constant() {r=0.5})
  n3 = Multiply(A=Power(Base=Multiply(A=@n2) {constB=2}), B=ScalarParameter() {defaultValue=0.75, parameterName=Edge Intensity}) {constB=0.65}
  NamedRerouteDeclaration(None=@n3) {name=Line Fades, nodeColor=<rgba 0.8516 0 1 1>}
  n4 = NamedRerouteUsage()
  n5 = MaterialFunctionCall(UV Channel=MaterialFunctionCall!(UV=Add(B=@n4) {constA=0.01, constB=-10}, Mirror Y, Swizzle UV) {materialFunction=/FortUILibrary/MaterialFunctions/Basics/MF_UI_Symmetry_2x1_New}.Symmetry) {materialFunction=&LinearGradient}
  n6 = Power(Base=@n5.UGradient) {constExponent=0.5}
  n7 = Power(Base=OneMinus(None=@n5.VGradient)) {constExponent=0.5}
  n8 = SmoothStep(Max=ScalarParameter() {defaultValue=0.9, parameterName=Underglow_Size}, Value=OneMinus(None=Multiply(A=@n6, B=@n7)))
  n9 = Power(Base=@n5.VGradient) {constExponent=0.2}
  n10 = Multiply(A=Min(A=Multiply(A=@n8, B=@n9)) {constB=0.999}, B=ScalarParameter() {defaultValue=1, parameterName=Underlight_Opacity})
  n11 = MaterialFunctionCall(UVs=@n4, Texture Scale=Constant2Vector() {g=1, r=4}) {materialFunction=&ScaleUVsByCenter}
  n12 = Constant() {r=1}
  n13 = Power(Base=OneMinus(None=Distance(A=@n11.UVs, B=@n12))) {constExponent=3}
  n14 = MaterialFunctionCall(UV=NamedRerouteUsage()) {materialFunction=&MF_UI_UVScaling}
  n15 = Constant() {r=0.5}
  n16 = Saturate(None=OneMinus(None=Distance(A=@n14.Scaled to Fit, B=@n15)))
  n17 = LinearInterpolate(
   A=LinearInterpolate(A=NamedRerouteUsage(), B=NamedRerouteUsage(), Alpha=Multiply(A=@n13) {constB=0.1})
   B=NamedRerouteUsage()
   Alpha=Multiply(
    A=MaterialFunctionCall("Value (0-1)"=@n16) {materialFunction=&MF_UI_EaseCurves}.EaseInQuint
    B=ScalarParameter() {parameterName=CenterGlow_Opacity}
   ) {constB=0}
  )
  n18 = Constant3Vector() {constant=<rgba 0.0395 1 0.2307 1>}
  n19 = ScalarParameter() {defaultValue=1, parameterName=GreenAmount}
  NamedRerouteDeclaration(None=LinearInterpolate(
    A=MaterialFunctionCall!(Blend=@n10, Base=@n17) {materialFunction=/Engine/Functions/Engine_MaterialFunctions03/Blends/Blend_ColorDodge}.Result
    B=Multiply(A=@n18, B=@n19) {constB=3}
    Alpha=NamedRerouteUsage()
   ) {constB=0}) {name=Modal Color, nodeColor=<rgba 0 0 0 1>}
  n20 = MaterialFunctionCall(
   UVs=MaterialFunctionCall() {materialFunction=&GetUserInterfaceUV}.Normalized UV
   Texture Scale=Constant2Vector() {g=2, r=1}
  ) {materialFunction=&ScaleUVsByCenter}
  n21 = AppendVector(A=OneMinus(None=ComponentMask(None=@n20.UVs) {r=true}), B=OneMinus(None=ComponentMask(None=@n20.UVs) {g=true}))
  n22 = Constant()
  n23 = Power(Base=OneMinus(None=Distance(A=@n21, B=@n22))) {constExponent=4}
  n24 = AppendVector(A=OneMinus(None=ComponentMask(None=@n20.UVs) {r=true}), B=ComponentMask(None=@n20.UVs) {g=true})
  n25 = Constant() {r=1}
  n26 = Power(Base=OneMinus(None=Distance(A=@n24, B=@n25))) {constExponent=3}
  NamedRerouteDeclaration(None=Multiply(A=LinearInterpolate(A=@n23, Alpha=@n26), B=ScalarParameter() {defaultValue=0.35, parameterName=Blobs Opacity}) {constB=0.3}) {name=Blobs, nodeColor=<rgba 0 0 0 1>}
  n27 = NamedRerouteUsage()
  n28 = NamedRerouteUsage()
  n29 = NamedRerouteUsage()
  NamedRerouteDeclaration(None=MaterialFunctionCall(Blend=Min(A=@n27) {constB=0.5}, Base=Multiply(A=@n28, B=@n29)) {materialFunction=&Blend_ColorDodge}.Result) {name=Line Colors, nodeColor=<rgba 0 0 0 1>}
  n30 = OneMinus(None=ScalarParameter() {defaultValue=0.1, parameterName=Line Reveal})
  n31 = LinearInterpolate(Alpha=@n30) {constA=-1}
  n32 = NamedRerouteUsage()
  n33 = Distance(A=ComponentMask(None=@n1.Normalized UV) {g=true}, B=Constant() {r=0.885})
  n34 = ScalarParameter() {defaultValue=4, parameterName=Line Thickness}
  n35 = OneMinus(None=Multiply(A=OneMinus(None=@n3), B=Multiply(A=@n33, B=@n34) {constB=5}))
  NamedRerouteDeclaration(None=Multiply(A=Power(Base=@n35) {constExponent=8}, B=SmoothStep(Min=@n30, Max=@n31, Value=@n32))) {name=Line Opacity, nodeColor=<rgba 0 0 0 1>}
  n36 = Constant3Vector() {constant=<rgba 0.0099 0.0083 0.024 1>}
  n37 = MaterialFunctionCall(GradientStart=Constant() {r=-0.5}, GradientEnd=Constant() {r=1.2}, RotationAngle=Constant() {r=0.3}) {materialFunction=&MF_UI_LinearGradient}
  NamedRerouteDeclaration(None=LinearInterpolate(
    A=Power(Base=@n36)
    B=@n36
    Alpha=MaterialFunctionCall("Value (0-1)"=@n37.Result) {materialFunction=&MF_UI_EaseCurves}.EaseInOutSine
   )) {name=Base Gradient, nodeColor=<rgba 0 0 0 1>}
  n38 = VectorParameter() {group=0 - Coloring, parameterName=Color 1, defaultValue=<rgba 0.5333 0.1981 1 1>}
  n39 = MaterialFunctionCall(UVs=TextureCoordinate(), Texture Scale=Constant() {r=2}) {materialFunction=&ScaleUVsByCenter}
  n40 = Frac(None=MaterialFunctionCall(UV Channel=Panner(Coordinate=@n39.UVs) {speedX=-0.2}) {materialFunction=&LinearGradient}.UGradient)
  n41 = Constant() {r=0.5}
  n42 = MaterialFunctionCall("Value (0-1)"=Multiply(A=Distance(A=@n40, B=@n41)) {constB=2}) {materialFunction=&MF_UI_EaseCurves}
  n43 = VectorParameter() {group=0 - Coloring, parameterName=Color 2, defaultValue=<rgba 0 1 1 1>}
  NamedRerouteDeclaration(None=Multiply(
    A=LinearInterpolate(A=@n38.RGB, B=@n43.RGB, Alpha=@n42.EaseInOutCubic)
    B=ScalarParameter() {defaultValue=1, parameterName=UnderGlow_Opacity}
   )) {name=Rift Gradient, nodeColor=<rgba 0 0 0 1>}
  n44 = Constant() {r=1}
  n45 = LinearInterpolate(
   B=ScalarParameter() {defaultValue=0.8, parameterName=Modal Size}
   Alpha=NamedRerouteDeclaration(None=ScalarParameter() {defaultValue=1, group=9 - Reveal Mask, parameterName=Modal Reveal Amount}) {name=Reveal Amount, nodeColor=<rgba 0 0 0 1>}
  ) {constA=0.01, constB=0.5}
  NamedRerouteDeclaration(None=MaterialFunctionCall(
    UVs=MaterialFunctionCall() {materialFunction=&GetUserInterfaceUV}.Normalized UV
    Texture Scale=AppendVector(A=@n44, B=@n45)
   ) {materialFunction=&ScaleUVsByCenter}.UVs) {name=ModalUV, nodeColor=<rgba 0 0 0 1>}
  n46 = NamedRerouteUsage()
  n47 = OneMinus(None=NamedRerouteUsage())
  n48 = Min(A=Step(Y=ComponentMask(None=@n46) {g=true}), B=Step(Y=ComponentMask(None=@n47) {g=true}))
  n49 = ComponentMask(None=NamedRerouteUsage()) {g=true}
  n50 = SmoothStep(Min=OneMinus(None=NamedRerouteUsage()), Value=@n49)
  n51 = LinearInterpolate(
   A=LinearInterpolate(B=ScalarParameter() {defaultValue=0.8, parameterName=Modal Top Fade}, Alpha=NamedRerouteUsage()) {constB=0.7}
   Alpha=@n50
  )
  n52 = Power(Base=SmoothStep(Value=@n49) {constMax=1.2, constMin=1}) {constExponent=0.4}
  n53 = NamedRerouteUsage()
  n54 = Multiply(A=Distance(A=ComponentMask(None=@n53) {r=true}, B=Constant() {r=0.5})) {constB=2}
  n55 = Power(Base=Multiply(A=OneMinus(None=@n52), B=Power(Base=@n54) {constExponent=1}))
  NamedRerouteDeclaration(None=LinearInterpolate(A=Multiply(A=@n48, B=@n51), Alpha=Multiply(A=@n50, B=@n55))) {name=Modal Mask, nodeColor=<rgba 0 0 0 1>}
  NamedRerouteUsage()
 ]
}
